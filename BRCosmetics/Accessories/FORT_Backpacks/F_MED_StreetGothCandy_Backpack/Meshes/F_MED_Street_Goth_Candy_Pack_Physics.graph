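# Physics asset for the Street Goth Candy backpack cosmetic, setting up dangling motion for its two ribbons and zipper pull.
# Capsule chains (four links per ribbon, three for the zipper) use limited swing and little or no twist.
# Each ribbon hangs from a small MSTR root body with free rotation on the box-shaped pack, while the zipper attaches directly.
# An unconstrained wide box, dyn_body_collision, appears to stand in for the wearer's body.
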PhysicsAsset /BRCosmetics/Accessories/FORT_Backpacks/F_MED_StreetGothCandy_Backpack/Meshes/F_MED_Street_Goth_Candy_Pack_Physics
bodies {
 dyn_body_collision [{shapeType=Box, extentX=55, extentY=3, extentZ=30, center=<0 7 -25>, rotation=<rot 0 0 13>}]
 dyn_l_ribbon_01 [{shapeType=Capsule, length=3, radius=1, center=<3 0 -0.5>, rotation=<rot 90 0 0>}]
 dyn_l_ribbon_01_MSTR [{shapeType=Capsule, length=0.505, radius=0.505, rotation=<rot 0 -90 0>}]
 dyn_l_ribbon_02 [{shapeType=Capsule, length=3, radius=1, center=<3 0 -0.5>, rotation=<rot 90 0 0>}]
 dyn_l_ribbon_03 [{shapeType=Capsule, length=3, radius=1, center=<3 0 -0.5>, rotation=<rot 90 0 0>}]
 dyn_l_ribbon_04 [{shapeType=Capsule, length=3, radius=1, center=<3 0 -0.5>, rotation=<rot 90 0 0>}]
 dyn_r_ribbon_01 [{shapeType=Capsule, length=3, radius=1, center=<3 0 -0.5>, rotation=<rot 90 0 0>}]
 dyn_r_ribbon_01_MSTR [{shapeType=Capsule, length=0.505, radius=0.505, rotation=<rot 0 -90 0>}]
 dyn_r_ribbon_02 [{shapeType=Capsule, length=3, radius=1, center=<3 0 -0.5>, rotation=<rot 90 0 0>}]
 dyn_r_ribbon_03 [{shapeType=Capsule, length=3, radius=1, center=<3 0 -0.5>, rotation=<rot 90 0 0>}]
 dyn_r_ribbon_04 [{shapeType=Capsule, length=3, radius=1, center=<3 0 -0.5>, rotation=<rot 90 0 0>}]
 dyn_zipper_01 [{shapeType=Capsule, length=3.0531, radius=1.1071, center=<2.3316 -0.0001 0>, rotation=<rot 90 0 0>}]
 dyn_zipper_02 [{shapeType=Capsule, length=3.0531, radius=1.1071, center=<2.3316 -0.0001 0>, rotation=<rot 90 0 0>}]
 dyn_zipper_03 [{shapeType=Capsule, length=3.0531, radius=1.1071, center=<2.3316 -0.0001 0>, rotation=<rot 90 0 0>}]
 pack [{shapeType=Box, extentX=34, extentY=8.4139, extentZ=23, center=<0 -5.2809 -5.5>}]
}
constraints [
 {bone1Name=dyn_l_ribbon_02, bone2Name=dyn_l_ribbon_01, swing1LimitDegrees=15, swing2LimitDegrees=30, twistLimitDegrees=2}
 {bone1Name=dyn_l_ribbon_03, bone2Name=dyn_l_ribbon_02, swing1LimitDegrees=15, swing2LimitDegrees=30, twistLimitDegrees=2}
 {bone1Name=dyn_r_ribbon_02, bone2Name=dyn_r_ribbon_01, swing1LimitDegrees=15, swing2LimitDegrees=30, twistLimitDegrees=2}
 {bone1Name=dyn_r_ribbon_03, bone2Name=dyn_r_ribbon_02, swing1LimitDegrees=15, swing2LimitDegrees=30, twistLimitDegrees=2}
 {bone1Name=dyn_zipper_01, bone2Name=pack, swing1LimitDegrees=60, swing2LimitDegrees=45, twistLimitDegrees=0}
 {bone1Name=dyn_zipper_02, bone2Name=dyn_zipper_01, swing1LimitDegrees=30, swing2LimitDegrees=45, twistLimitDegrees=0}
 {bone1Name=dyn_zipper_03, bone2Name=dyn_zipper_02, swing1LimitDegrees=30, swing2LimitDegrees=45, twistLimitDegrees=0}
 {bone1Name=dyn_r_ribbon_04, bone2Name=dyn_r_ribbon_03, swing1LimitDegrees=15, swing2LimitDegrees=30, twistLimitDegrees=2}
 {bone1Name=dyn_l_ribbon_04, bone2Name=dyn_l_ribbon_03, swing1LimitDegrees=15, swing2LimitDegrees=30, twistLimitDegrees=2}
 {bone1Name=dyn_r_ribbon_01, bone2Name=dyn_r_ribbon_01_MSTR, swing1LimitDegrees=15, swing2LimitDegrees=30, twistLimitDegrees=2}
 {bone1Name=dyn_l_ribbon_01, bone2Name=dyn_l_ribbon_01_MSTR, swing1LimitDegrees=15, swing2LimitDegrees=30, twistLimitDegrees=2}
 {
  bone1Name=dyn_l_ribbon_01_MSTR
  bone2Name=pack
  swing1LimitDegrees=45
  swing1Motion=Free
  swing2LimitDegrees=45
  swing2Motion=Free
  twistLimitDegrees=45
  twistMotion=Free
 }
 {
  bone1Name=dyn_r_ribbon_01_MSTR
  bone2Name=pack
  swing1LimitDegrees=45
  swing1Motion=Free
  swing2LimitDegrees=45
  swing2Motion=Free
  twistLimitDegrees=45
  twistMotion=Free
 }
]
