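# Anim Blueprint for the Street Goth Candy backpack, adding secondary motion to its dangling bones.
# bRigidBodyEnabled selects between RigidBody physics, preceded by additive ribbon rotation and collision-bone offset, and Trail Controllers with planar limits for both ribbons and the zipper.
# Each update sets those pre-physics alphas from pawn speed and crouch or crouch-sprint state.
# It also derives backpack_rotate from spine twist, though the AnimGraph never reads it.
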AnimBlueprint /BRCosmetics/Accessories/FORT_Backpacks/F_MED_StreetGothCandy_Backpack/Meshes/F_MED_Street_Goth_Candy_Pack_AnimBlueprint
target_skeleton=&/F_MED_Street_Goth_Candy_Pack_Skeleton
graphs [
 AnimGraph {nodes [
   n1 = Trail controller - Bone: dyn_l_ribbon_04(ComponentPose=Local To Component(LocalPose=Use cached pose 'input_cache'()), Alpha=1) {
    bUsePlanarLimit=true
    chainLength=4
    maxDeltaTime=0.0333
    relaxationSpeedScale=3
    planarLimits [
     {drivingBone {boneName=dyn_l_ribbon_01_MSTR}, planeTransform {location=<0 0 0>, rotation=<rot 10 0 0>, scale=<1 1 1>}}
     {drivingBone {boneName=dyn_l_ribbon_01_MSTR}, planeTransform {location=<2 0 0>, rotation=<rot -90 0 0>, scale=<1 1 1>}}
    ]
    rotationLimits [{limitMax=<180 180 180>}, {limitMax=<180 180 180>}, {limitMax=<180 180 180>}, {limitMax=<180 180 180>}]
    rotationOffsets [<0 0 0>, <0 0 0>, <0 0 0>, <0 0 0>]
    trailBone {boneName=dyn_l_ribbon_04}
    trailRelaxationSpeed {editorCurveData {keys [<at 0 10 RCIM_Linear>,<at 0.5 9 RCIM_Linear>,<at 1 5 RCIM_Linear>]}}
   }
   n2 = Trail controller - Bone: dyn_r_ribbon_04(ComponentPose=@n1, Alpha=1) {
    bUsePlanarLimit=true
    chainLength=4
    maxDeltaTime=0.0333
    relaxationSpeedScale=3
    planarLimits [
     {drivingBone {boneName=dyn_r_ribbon_01_MSTR}, planeTransform {location=<0 0 0>, rotation=<rot 10 0 0>, scale=<1 1 1>}}
     {drivingBone {boneName=dyn_r_ribbon_01_MSTR}, planeTransform {location=<2 0 0>, rotation=<rot -90 0 0>, scale=<1 1 1>}}
    ]
    rotationLimits [{limitMax=<180 180 180>}, {limitMax=<180 180 180>}, {limitMax=<180 180 180>}, {limitMax=<180 180 180>}]
    rotationOffsets [<0 0 0>, <0 0 0>, <0 0 0>, <0 0 0>]
    trailBone {boneName=dyn_r_ribbon_04}
    trailRelaxationSpeed {editorCurveData {keys [<at 0 10 RCIM_Linear>,<at 0.5 9 RCIM_Linear>,<at 1 5 RCIM_Linear>]}}
   }
   n3 = Trail controller - Bone: dyn_zipper_03(ComponentPose=@n2, Alpha=1) {
    bUsePlanarLimit=true
    chainLength=3
    maxDeltaTime=0.0333
    relaxationSpeedScale=4
    planarLimits [{drivingBone {boneName=pack}, planeTransform {location=<0 -11 0>, rotation=<rot 0 0 -90>, scale=<1 1 1>}}]
    rotationLimits [{limitMax=<180 180 180>}, {limitMax=<180 180 180>}, {limitMax=<180 180 180>}]
    rotationOffsets [<0 0 0>, <0 0 0>, <0 0 0>]
    trailBone {boneName=dyn_zipper_03}
    trailRelaxationSpeed {editorCurveData {keys [<at 0 10 RCIM_Linear>,<at 0.5 9 RCIM_Linear>,<at 1 5 RCIM_Linear>]}}
   }
   n4 = Get RibbonRotateAlpha() -> (RibbonRotateAlpha:double)
   n5 = "Transform (Modify) Bone - Bone: dyn_l_ribbon_01_MSTR"(
    ComponentPose=Local To Component(LocalPose=Use cached pose 'input_cache'())
    Alpha=@n4
    Translation="0,0,0"
    Rotation="45,0.000,0.000"
    Scale="1,1,1"
   ) {
    rotationMode=BMM_Additive
    rotationSpace=BCS_BoneSpace
    translationMode=BMM_Additive
    translationSpace=BCS_BoneSpace
    alphaBoolBlend {blendOption=Linear}
    alphaScaleBiasClamp {bInterpResult=true, interpSpeedDecreasing=2}
    boneToModify {boneName=dyn_l_ribbon_01_MSTR}
    rotation {pitch=45}
   }
   n6 = "Transform (Modify) Bone - Bone: dyn_r_ribbon_01_MSTR"(ComponentPose=@n5, Alpha=@n4, Translation="0,0,0", Rotation="45,0,0", Scale="1,1,1") {
    rotationMode=BMM_Additive
    rotationSpace=BCS_BoneSpace
    translationMode=BMM_Additive
    translationSpace=BCS_BoneSpace
    alphaBoolBlend {blendOption=Linear}
    alphaScaleBiasClamp {bInterpResult=true, interpSpeedDecreasing=2}
    boneToModify {boneName=dyn_r_ribbon_01_MSTR}
    rotation {pitch=45}
   }
   n7 = "Transform (Modify) Bone - Bone: dyn_body_collision"(
    ComponentPose=@n6
    Alpha=Get CollsionTransformAlpha() -> (CollsionTransformAlpha:double)
    Translation="0,-8,0"
    Rotation="0,0,0"
    Scale="1,1,1"
   ) {
    translationMode=BMM_Additive
    translationSpace=BCS_BoneSpace
    alphaBoolBlend {blendOption=Linear}
    alphaScaleBiasClamp {bInterpResult=true, interpSpeedDecreasing=2}
    boneToModify {boneName=dyn_body_collision}
    translation {y=-8}
   }
   n8 = RigidBody(ComponentPose=@n7, Alpha=Get RigidAlpha() -> (RigidAlpha:double)) {
    bClampLinearTranslationLimitToRefPose=true
    bOverrideWorldGravity=true
    overridePhysicsAsset=&/F_MED_Street_Goth_Candy_Pack_Physics
    alphaBoolBlend {blendOption=Linear}
    componentAppliedLinearAccClamp=<3000 3000 3000>
    componentLinearAccScale=<0.3 0.3 0.4>
    componentLinearVelScale=<0.3 0.3 0.3>
    overrideWorldGravity {z=-980}
   }
   Output Pose(Result=Blend Poses by bool(BlendPose_0=Component To Local(ComponentPose=@n8), BlendPose_1=Component To Local(ComponentPose=@n3), bActiveValue=Get bRigidBodyEnabled()) {blendType=Linear}) {layerGroup=None}
   AnimGraphNode_SubInput_0 = Input Pose()
   Save cached pose 'input_cache'(Pose=@AnimGraphNode_SubInput_0) {cacheName=input_cache}
  ]}
 EventGraph {nodes [
   Event BlueprintUpdateAnimation()
   > n1 = Sequence()
   > Set CollsionTransformAlpha(CollsionTransformAlpha:double=MapRangeClamped(Value=Get PawnSpeed2D(), InRangeA=150, InRangeB=300)) -> (Output_Get:double)
   > n2 = Branch(Condition=Get bIsCrouching())
   > Set CollsionTransformAlpha(CollsionTransformAlpha=0.2)
   > Set RibbonRotateAlpha(RibbonRotateAlpha:double=0.8) -> (Output_Get:double)
   > n3 = Branch(Condition=Get bIsCrouchSprinting())
   > Set CollsionTransformAlpha(CollsionTransformAlpha=-0.2)
   > Set RibbonRotateAlpha(RibbonRotateAlpha=0.4)
   Set CollsionTransformAlpha(execute=@n3.else, CollsionTransformAlpha=0.2)
   > Set RibbonRotateAlpha(RibbonRotateAlpha=0.8)
   Set RibbonRotateAlpha(execute=@n2.else, RibbonRotateAlpha=1)
   n4 = GetOwnerSkeletalMeshComponent()
   Set backpack_rotate(
    execute=@n1.then_1
    backpack_rotate=MakeRotator(Yaw=GetAlphaFromReferenceTransformDelta(
      SkeletalMeshComponent=@n4
      BoneName=spine_05
      BaseName=spine_01
      DeltaComponentType=QuaternionTwist
      InputRange=(X=12,Y=-30)
      OutputRange=(X=5,Y=10)
      TwistAxis="1,0,0"
     ).RawOutput)
   )
   TryGetPawnOwner()
  ]}
]
variables [
 backpack_rotate {type=Rotator}
 CollsionTransformAlpha {type=number}
 RibbonRotateAlpha {type=number, default=1}
 RigidAlpha {type=number, default=1}
]
